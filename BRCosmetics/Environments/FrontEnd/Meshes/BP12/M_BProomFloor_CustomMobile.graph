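# Lit floor material for the BP12 room in the front end, with a custom brightness path for mobile.
# It samples diffuse, mask and normal floor textures.
# A feature level switch passes the diffuse RGB by default, while ES2 and ES3_1 take only its red channel multiplied by the MobileBrightness parameter, giving a brightness-adjusted greyscale value.
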Material /BRCosmetics/Environments/FrontEnd/Meshes/BP12/M_BProomFloor_CustomMobile
graph {
 properties {pixelDepthOffsetMode=PDOM_Legacy, usedShadingModels=Default Lit}
 nodes [
  n1 = TextureSample() {texture=/BRCosmetics/Environments/FrontEnd/Meshes/BP12/Textures/T_Floor_01_BP12_D}
  n2 = Multiply(A=ScalarParameter() {defaultValue=3.1408, parameterName=MobileBrightness}, B=@n1.R)
  FeatureLevelSwitch(Default=@n1.RGB, ES2=@n2, ES3_1=@n2)
  TextureSample() {samplerType=SAMPLERTYPE_Masks, texture=/BRCosmetics/Environments/FrontEnd/Meshes/BP12/Textures/T_Floor_01_BP12_S}
  TextureSample() {samplerType=SAMPLERTYPE_Normal, texture=/BRCosmetics/Environments/FrontEnd/Meshes/BP12/Textures/T_Floor_01_BP12_N}
 ]
}
