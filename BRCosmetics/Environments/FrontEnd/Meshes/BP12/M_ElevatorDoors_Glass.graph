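# Translucent, lit glass material for the elevator doors in the BP12 front-end environment, with shader-driven opening.
# Open and OpenMouseOver drive a Z rotation about the mesh origin up to RotationMax, with FixRotateAboutAxisNormals for normals.
# Vertex color red splits the rotation direction, swappable by the Flip switch and biased on one side by RotationMaxBias.
# A glass pattern texture blends two opacity values, alongside a tint color and a world-Z sine stripe mask.
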Material /BRCosmetics/Environments/FrontEnd/Meshes/BP12/M_ElevatorDoors_Glass
graph {
 properties {
  bUsedWithSplineMeshes=true
  blendMode=BLEND_Translucent
  pixelDepthOffsetMode=PDOM_Legacy
  translucencyLightingMode=TLM_SurfacePerPixelLighting
  usedShadingModels=Default Lit
 }
 nodes [
  n1 = Divide(A=ComponentMask(None=WorldPosition().XYZ) {b=true}, B=ScalarParameter() {defaultValue=12.1431, parameterName=Width})
  Ceil(None=Saturate(None=Sine(None=@n1)))
  LinearInterpolate(
   A=ScalarParameter() {defaultValue=0.0169, parameterName=Opacity}
   B=ScalarParameter() {defaultValue=0.2571, parameterName=Opacity2}
   Alpha=TextureSample() {texture=/BRCosmetics/Environments/FrontEnd/Meshes/BP12/Textures/T_GlassPattern_01_D}.R
  )
  n2 = Constant3Vector() {constant=<rgba 0 0 1 0>}
  n3 = ScalarParameter() {parameterName=Open}
  n4 = ScalarParameter() {parameterName=OpenMouseOver}
  n5 = LinearInterpolate(B=ScalarParameter() {defaultValue=0.06, parameterName=RotationMax}, Alpha=Saturate(None=Add(A=@n3, B=@n4)))
  n6 = VertexColor()
  n7 = LinearInterpolate(
   A=Multiply(A=@n5, B=ScalarParameter() {defaultValue=1, parameterName=RotationMaxBias})
   B=Multiply(A=@n5) {constB=-1}
   Alpha=StaticSwitchParameter(True=@n6.R, False=OneMinus(None=@n6.R)) {parameterName=Flip}
  )
  MaterialFunctionCall(Rotation Axis=@n2, Rotation Angle=@n7) {materialFunction=&FixRotateAboutAxisNormals}
  Max(A=@n3, B=@n4)
  RotateAboutAxis(
   NormalizedRotationAxis=@n2
   RotationAngle=@n7
   PivotPoint=TransformPosition(None=Constant3Vector()) {transformType=TRANSFORMPOSSOURCE_World}
   Position=WorldPosition().XYZ
  )
  ScalarParameter() {parameterName=M}
  ScalarParameter() {defaultValue=1, parameterName=S}
  ScalarParameter() {parameterName=R}
  VectorParameter() {parameterName=Color, defaultValue=<rgba 0.3199 0.3457 0.4583 1>}
 ]
}
