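# BP12_minimap is a masked, Default Lit material for a BP12 front-end environment mesh, apparently an objective board map.
# It samples the T_ObjectiveBoard_Map_D texture into a lerp whose B and Alpha are zero constants labelled as texture placeholders, so the map passes through unchanged.
# Three standalone constants sit alongside the lerp, and the graph shows no links from any node to material outputs.
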Material /BRCosmetics/Environments/FrontEnd/Meshes/BP12/Materials/BP12_minimap
graph {
 properties {blendMode=BLEND_Masked, pixelDepthOffsetMode=PDOM_Legacy, usedShadingModels=Default Lit}
 nodes [
  Constant() {r=0.2}
  Constant() {r=1}
  Constant()
  LinearInterpolate(
   A=TextureSample() {texture=/BRCosmetics/Environments/FrontEnd/Meshes/BP12/Textures/T_ObjectiveBoard_Map_D}.RGB
   B=Constant() {desc=Texture here}
   Alpha=Constant() {desc=Texture alpha here}
  )
 ]
}
