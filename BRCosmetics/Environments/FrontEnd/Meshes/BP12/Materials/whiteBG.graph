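# whiteBG is a simple Default Lit material in the FrontEnd BP12 mesh materials, likely for a white background surface by its name.
# Its graph holds only four scalar constants, with no textures, parameters or math.
# The file does not show which material inputs those constants feed.
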Material /BRCosmetics/Environments/FrontEnd/Meshes/BP12/Materials/whiteBG
graph {
 properties {pixelDepthOffsetMode=PDOM_Legacy, usedShadingModels=Default Lit}
 nodes [
  Constant() {r=0.25}
  Constant() {r=0.7}
  Constant() {r=0.5}
  Constant() {r=0.3}
 ]
}
